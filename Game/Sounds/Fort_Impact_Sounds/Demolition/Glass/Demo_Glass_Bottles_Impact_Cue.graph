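# Sound cue for glass bottle impacts in demolition, routed to the Fort impacts sound class.
# A mixer layers two random picks at equal volume: one of four glass bottle impacts and one of three shelf bottle impacts.
# Each layer has its own slight pitch modulation for variation.
# Small-object demolition attenuation shapes falloff, and a shared object-impact concurrency set limits simultaneous voices.
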SoundCue /Game/Sounds/Fort_Impact_Sounds/Demolition/Glass/Demo_Glass_Bottles_Impact_Cue
root=Mixer_2
nodes [
 {id=Mixer_2, children [Modulator_1,Modulator_4], properties {inputVolume [1,1]}}
 {id=Modulator_1, children [Random_2], properties {pitchMax=1.07, pitchMin=0.92}}
 {id=Modulator_4, children [Random_6], properties {pitchMax=1.07, pitchMin=0.92}}
 {id=Random_2, children [WavePlayer_8, WavePlayer_9, WavePlayer_10, WavePlayer_11], properties {weights [1,1,1,1]}}
 {id=Random_6, children [WavePlayer_22, WavePlayer_23, WavePlayer_24], properties {weights [1,1,1]}}
 {id=WavePlayer_10, properties {soundWaveAssetPtr=&/Glass_Bottle_Impact03}}
 {id=WavePlayer_11, properties {soundWaveAssetPtr=&/Glass_Bottle_Impact04}}
 {id=WavePlayer_22, properties {soundWaveAssetPtr=/Game/Sounds/Fort_Impact_Sounds/Demolition/Shelves/Shelf_Bottle_Impact02}}
 {id=WavePlayer_23, properties {soundWaveAssetPtr=/Game/Sounds/Fort_Impact_Sounds/Demolition/Shelves/Shelf_Bottle_Impact05}}
 {id=WavePlayer_24, properties {soundWaveAssetPtr=/Game/Sounds/Fort_Impact_Sounds/Demolition/Shelves/Shelf_Bottle_Impact06}}
 {id=WavePlayer_8, properties {soundWaveAssetPtr=&/Glass_Bottle_Impact01}}
 {id=WavePlayer_9, properties {soundWaveAssetPtr=&/Glass_Bottle_Impact02}}
]
properties {
 attenuationSettings=/Game/Sounds/Attenuation/Fort_Demolition_Hit_Small_Sized_Object_Att
 maxDistance=4000
 soundClassObject=&Fort_Impacts
 volumeMultiplier=0.68
 concurrencyOverrides {maxCount=3, resolutionRule=StopOldest}
 concurrencySet [&CONC_Object_Impacts_FourMax]
}
