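# Sound cue for demolition glass impacts, playing one of four window-break waves at random with randomized volume.
# An IsLocal branch has audio only on its true input, through a mixer and modulator into the random node; false and unset inputs are empty.
# It uses the impacts sound class, close-hit crossfade attenuation and a concurrency override that stops the oldest beyond two instances, despite also referencing a four-max concurrency set.
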SoundCue /Game/Sounds/Fort_Impact_Sounds/Demolition/Glass/SC_Demo_Default_Glass_Impact
root=Branch_0
nodes [
 {id=Branch_0, children [Mixer_3,null,null], properties {boolParameterName=IsLocal}}
 {id=Mixer_3, children [Modulator_1], properties {inputVolume [1]}}
 {id=Modulator_1, children [Random_0], properties {volumeMax=1.2, volumeMin=1}}
 {id=Random_0, children [WavePlayer_0,WavePlayer_1,WavePlayer_2,WavePlayer_3], properties {weights [1,1,1,1]}}
 {id=WavePlayer_0, properties {soundWaveAssetPtr=&/Glass_Window_Break_Demo_Impact_01}}
 {id=WavePlayer_1, properties {soundWaveAssetPtr=&/Glass_Window_Break_Demo_Impact_02}}
 {id=WavePlayer_2, properties {soundWaveAssetPtr=&/Glass_Window_Break_Demo_Impact_03}}
 {id=WavePlayer_3, properties {soundWaveAssetPtr=&/Glass_Window_Break_Demo_Impact_04}}
]
properties {
 attenuationSettings=/Game/Sounds/Attenuation/Fort_Demolition_Crossfade_Global_Hit_Close_Att
 bOverrideConcurrency=true
 maxDistance=4800
 soundClassObject=&Fort_Impacts
 volumeMultiplier=1
 concurrencyOverrides {maxCount=2, resolutionRule=StopOldest}
 concurrencySet [&CONC_Object_Impacts_FourMax]
}
